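; Gameplay ability for the ballistic shield that plays the small bash montage.
; On activation it runs the parent logic, then starts the MontageToPlay montage through a PlayMontageAndWait task with no callbacks wired.
; The ability ends immediately after starting the task, with stop-on-end disabled so the montage outlives it.
; CanApplyHolster always returns false.
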(event Ability|EventActivateAbility
  (|CallParentFunction)
  (Ability|Tasks|PlayMontageAndWait "None" (Variables|Default|GetMontageToPlay) 1.0 "None" false
    (:OnCompleted
      _)
    (:OnBlendOut
      _)
    (:OnInterrupted
      _)
    (:OnCancelled
      _))
  (Ability|EndAbility))

(fn CanApplyHolster
  (params (HolsterId :type Name%20%28by%20ref%29 :default "None"))
  (return false))
